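# Flare material in a cat laser test folder, set to translucent unlit deferred decal, with no material output connections shown.
# Color chain layers sphere-mask center and corona with a flare texture, tinted and optionally scaled by flicker and time of day.
# Opacity chain combines sine flicker, depth fade, near-camera fade and optional masks.
# Offset chain builds a camera-facing sprite with optional distance scaling and a sine-driven sway about an axis.
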Material /CosmeticCompanions/Assets/Quadruped/FondCuddles/laserTest/CatLaserTest_Player/M_Flare_01
graph {
 properties {
  bUsedWithNiagaraMeshParticles=true
  blendMode=BLEND_Translucent
  materialDomain=MD_DeferredDecal
  pixelDepthOffsetMode=PDOM_Legacy
  shadingModel=MSM_Unlit
  twoSided=true
  usedShadingModels=Unlit
 }
 nodes [
  n1 = VectorParameter() {parameterName=Param_2, defaultValue=<rgba 1 1 0 1>}
  AppendVector(A=@n1.R, B=@n1.R)
  MaterialFunctionCall() {materialFunction=&RadialGradientExponential}
  MaterialFunctionCall()
  n2 = VectorParameter() {parameterName=Param_1, defaultValue=<rgba 1 1 1 1>}
  MaterialFunctionCall(R=@n2.R, G=@n2.R, B=@n2.R) {materialFunction=&AppendMany}
  MaterialFunctionCall!(
   Center Location
   0-1 UVs=TextureCoordinate() {coordinateIndex=1}
   XY scale=ScalarParameter() {defaultValue=125, parameterName=Sprite Size}
   OverridePivot
   "Normals (Optional)"
  ) {materialFunction=&Sprite}
  MaterialFunctionCall() {materialFunction=&ComponentPivotLocation}
  ScalarParameter() {defaultValue=1, parameterName=WidthScale}
  n3 = Frac(None=MaterialFunctionCall(Speed=ScalarParameter() {defaultValue=1, parameterName=Flicker_PulseFrequence}) {materialFunction=&TimeWithSpeedVariable}.Time)
  n4 = ScalarParameter() {defaultValue=1, parameterName=Flicker_FrequenceBlack}
  n5 = Clamp(None=Ceil(None=Subtract(A=@n3, B=@n4)))
  n6 = Time()
  n7 = MaterialFunctionCall(Float3=ActorPositionWS()) {materialFunction=&BreakOutFloat3Components}
  n8 = Add(A=Add(A=@n7.R, B=@n7.R), B=@n7.B)
  n9 = LinearInterpolate(A=Multiply(A=@n6) {constB=0.7}, B=Multiply(A=@n6) {constB=1.5}, Alpha=Frac(None=Divide(A=@n8) {constB=50}))
  n10 = Sine(None=@n9) {period=0.5651}
  n11 = Sine(None=@n9) {period=0.442}
  n12 = Add(A=Multiply(A=Add(A=@n10, B=@n11), B=ScalarParameter() {defaultValue=0.15, group=Emissive, parameterName=GlowFlickerScale}))
  n13 = Constant() {r=0.5}
  n14 = SphereMask(A=TextureCoordinate(), B=@n13, Radius=@n13) {hardnessPercent=0}
  n15 = Constant() {r=0.5}
  n16 = ParticleRandom()
  n17 = Multiply(
   A=Time()
   B=StaticSwitchParameter(True=LinearInterpolate(Alpha=@n16) {constA=1.2, constB=0.8}, False=Constant() {r=1}) {parameterName=GPU Particles}
  )
  n18 = Sine(None=@n17) {period=0.075}
  n19 = Multiply(A=Multiply(A=Sine(None=@n17) {period=0.155}, B=Sine(None=@n17) {period=0.2}), B=Sine(None=@n17) {period=0.43})
  n20 = ScalarParameter() {defaultValue=0.4, desc="0 = full flicker amount", parameterName=GPUFlickerFadeMin}
  n21 = ScalarParameter() {defaultValue=1, desc="1 = no flicker", parameterName=GPUFlickerFadeMax, sortPriority=1}
  n22 = ParticleRandom()
  n23 = ScalarParameter() {defaultValue=1, parameterName=FlickerLerp}
  n24 = DynamicParameter() {paramNames [FlickerFade,CoronaSize,Param3,Size]}
  n25 = LinearInterpolate(
   A=Clamp(None=Multiply(A=@n18, B=@n19)) {maxDefault=3}
   Alpha=StaticSwitchParameter(True=LinearInterpolate(A=@n20, B=@n21, Alpha=@n22) {constA=0.5}, False=Multiply(A=@n23, B=@n24.FlickerFade)) {parameterName=GPU Particles}
  ) {constA=1}
  n26 = ScalarParameter() {defaultValue=1, parameterName=CoronaSize}
  n27 = Add(
   A=LinearInterpolate(Alpha=@n25) {constA=-0.4, constB=0}
   B=StaticSwitchParameter(True=@n26, False=Multiply(A=@n24.CoronaSize, B=@n26)) {parameterName=GPU Particles}
  ) {constB=0}
  n28 = TextureSample(UVs=SphereMask(A=TextureCoordinate(), B=@n15, Radius=@n15, Hardness=LinearInterpolate(Alpha=@n27) {constA=0.7, constB=-0.3}) {hardnessPercent=25}) {texture=/Game/Effects/Fort_Effects/Textures/Grad/T_Halo_Corona_Gradient}
  n29 = Constant() {r=1}
  n30 = Multiply(
   A=StaticSwitchParameter(True=Desaturation(None=@n28.RGB, Fraction=@n29), False=@n28.RGB) {parameterName=UseDesaturatedFlare?}
   B=ScalarParameter() {defaultValue=0.125, parameterName=CoronaBrightness}
  ) {constB=0.125}
  n31 = Add(A=Multiply(A=Power(Base=@n14) {constExponent=25}, B=ScalarParameter() {defaultValue=2, parameterName=CenterBrightness}) {constB=2}, B=@n30)
  n32 = VectorParameter() {parameterName=Color, defaultValue=<rgba 1 1 1 1>}
  n33 = ParticleColor()
  n34 = StaticSwitchParameter(True=@n32.RGB, False=Multiply(A=@n33.RGB, B=@n32.RGB)) {parameterName=UsedWithSkeletalMesh}
  n35 = Multiply(A=Power(Base=@n14) {constExponent=4}, B=ScalarParameter() {defaultValue=0.15, parameterName=EdgeBrightness}) {constB=0.15}
  n36 = TextureSampleParameter2D() {parameterName=FlareMask, texture=/Game/Effects/Fort_Effects/Textures/Flares/T_Flare_Spikeball}
  n37 = Add(
   A=Multiply(A=ShadingPathSwitch(Default=@n31, Forward=@n30), B=LinearInterpolate(B=@n34) {constA=1, constAlpha=0.7})
   B=Multiply(A=Add(A=@n35, B=@n36.RGB), B=@n34)
  )
  n38 = StaticSwitchParameter(True=Multiply(A=StaticSwitchParameter(True=@n5, False=@n12) {parameterName=UseSimplePulseFlicker}, B=@n37), False=@n37) {parameterName=UseGlowFlicker?}
  n39 = ScalarParameter() {defaultValue=1, parameterName="Value_Morning (S)"}
  n40 = ScalarParameter() {defaultValue=1, parameterName="Value_Day (S)"}
  n41 = ScalarParameter() {defaultValue=1, parameterName="Value_Evening (S)"}
  n42 = ScalarParameter() {defaultValue=1, parameterName="Value_Night (S)"}
  StaticSwitchParameter(
   True=Multiply(
    A=@n38
    B=MaterialFunctionCall(Value_Morning=@n39, Value_Day=@n40, Value_Evening=@n41, Value_Night=@n42) {materialFunction=&Scalar_Time_of_Day}.Result
   )
   False=@n38
  ) {parameterName=UseTOD_Intensity?}
  n43 = Add(
   A=TextureCoordinate()
   B=ComponentMask(None=VectorParameter() {parameterName=MaskDirection, defaultValue=<rgba 0 -0.5 0 1>}.RGB) {g=true, r=true}
  )
  n44 = ScalarParameter() {defaultValue=10, parameterName=NearCamFadeDistance}
  n45 = TextureSampleParameter2D() {parameterName=OuterFlareMaskTexture, samplerType=SAMPLERTYPE_Grayscale, texture=/Game/Effects/Fort_Effects/Textures/T_EFX_CameraFlare_01}
  n46 = StaticSwitchParameter(True=@n32.A, False=Multiply(A=@n33.A, B=@n32.A)) {parameterName=UsedWithSkeletalMesh}
  n47 = DepthFade(Opacity=@n46, FadeDistance=ScalarParameter() {defaultValue=20, parameterName=DepthFade})
  n48 = Power(Base=MaterialFunctionCall(Distance=ScalarParameter() {defaultValue=150, parameterName=NearCameraShrink}) {materialFunction=&Near_Camera_Fade}.Result) {constExponent=3}
  n49 = Multiply(A=@n25, B=Multiply(A=FeatureLevelSwitch(Default=@n47, ES2=@n46, ES3_1=@n46, SM4_REMOVED=@n47, SM5=@n47, SM6=@n47), B=@n48))
  n50 = StaticSwitchParameter(True=Multiply(A=@n45.G, B=@n49), False=@n49) {parameterName=UseFlareMask}
  n51 = CollectionParameter() {
   collection=&FortniteMaterialParameters
   parameterName=FortniteActiveTimeOfDay
  }
  n52 = DotProduct(A=@n51, B=Constant4Vector() {constant=<rgba 0 0 0 1>}.RGBA)
  n53 = Multiply(
   A=StaticSwitchParameter(
    True=MaterialFunctionCall(Distance=@n44) {materialFunction=&Near_Camera_Fade}.Result
    False=Constant() {r=1}
   ) {parameterName=UseNearCamFade?}
   B=StaticSwitchParameter(True=LinearInterpolate(B=@n50, Alpha=@n52), False=@n50) {parameterName=Turn Off At Day}
  )
  StaticSwitchParameter(True=Multiply(A=ComponentMask(None=@n43) {g=true}, B=@n53), False=@n53) {parameterName=UseUpDown_Mask?}
  n54 = DotProduct(
   A=CameraVectorWS()
   B=TransformPosition(None=VectorParameter() {parameterName=Axis, defaultValue=<rgba 0 0 1 0>}.RGB) {transformSourceType=TRANSFORMPOSSOURCE_World}
  )
  StaticSwitchParameter(True=LinearInterpolate(Alpha=Saturate(None=@n54))) {parameterName=Camera Offset Angle}
  n55 = Constant3Vector() {constant=<rgba 0 0 1 0>}
  n56 = ScalarParameter() {parameterName=rotation}
  n57 = Multiply(A=Constant() {r=34.9111}, B=Frac(None=Max(A=@n56) {constB=0})) {constA=16}
  n58 = Constant3Vector()
  n59 = VectorParameter() {parameterName=Rotation Vector, defaultValue=<rgba 1 0 0 1>}
  n60 = StaticSwitchParameter(
   True=Transform(None=Constant3Vector() {constant=<rgba 0 0 -1 0>}) {transformSourceType=TRANSFORMSOURCE_Local}
   False=Normalize(VectorInput=RotateAboutAxis(NormalizedRotationAxis=@n55, RotationAngle=@n57, PivotPoint=@n58, Position=@n59.RGB))
  ) {group=Creative Switch, parameterName=Is Creative, sortPriority=0}
  n61 = Time()
  n62 = LinearInterpolate(
   A=ScalarParameter() {defaultValue=0.125, parameterName=Rotation Speed}
   B=ScalarParameter() {defaultValue=0.07, parameterName=Rotation Speed2}
   Alpha=Frac(None=@n57)
  )
  n63 = Multiply(A=Sine(None=Multiply(A=@n61, B=@n62)), B=ScalarParameter() {defaultValue=0.125, parameterName=Rotation Max})
  n64 = VectorParameter() {parameterName=Pivot Offset Custom, defaultValue=<rgba 0 0 75 1>}
  n65 = Add(
   A=ActorPositionWS()
   B=StaticSwitchParameter(True=Transform(None=@n64.RGB) {transformSourceType=TRANSFORMSOURCE_Local}, False=@n64.RGB) {group=Creative Switch, parameterName=Is Creative, sortPriority=0}
  )
  n66 = WorldPosition()
  n67 = ScalarParameter() {defaultValue=30, parameterName=SpriteScale}
  n68 = MaterialFunctionCall(Distance=ScalarParameter() {defaultValue=5000, group=WPO, parameterName=SpriteCamGrowDistance}) {materialFunction=&Near_Camera_Fade}
  n69 = StaticSwitchParameter(
   True=LinearInterpolate(A=@n67, B=ScalarParameter() {defaultValue=30, parameterName=SpriteScaleMax}, Alpha=Power(Base=@n68.Result) {constExponent=4})
   False=@n67
  ) {parameterName=UseMaxDistanceScaling?}
  n70 = CollectionParameter() {
   collection=&FortniteMaterialParameters
   parameterName=FortniteActiveTimeOfDay
  }
  n71 = Multiply(A=@n69, B=DotProduct(A=@n70, B=Constant4Vector() {constant=<rgba 0 0 0 1>}.RGBA))
  n72 = Multiply(A=@n24.Size, B=Multiply(A=StaticSwitchParameter(True=@n71, False=@n69) {parameterName=Turn Off At Day}, B=@n48))
  n73 = VectorParameter()
  n74 = ScalarParameter() {parameterName=CameraOffset}
  n75 = ObjectPositionWS()
  n76 = WorldPosition()
  n77 = Multiply(A=ScalarParameter() {defaultValue=10, parameterName=MaxScaleIncrease}, B=Multiply(A=Subtract(A=@n75, B=@n76.XYZ)) {constB=-1}) {constB=-1}
  n78 = Distance(A=@n75, B=CameraPositionWS())
  n79 = ScalarParameter() {defaultValue=2000, parameterName=DistanceOffset}
  n80 = Clamp(None=Divide(A=Subtract(A=@n78, B=@n79), B=ScalarParameter() {defaultValue=30000, parameterName=ScaleUpDistance}))
  n81 = StaticSwitchParameter(
   True=Add(
    A=MaterialFunctionCall(Center Location=@n66.XYZ, XY scale=@n72, "Normals (Optional)"=@n73.RGB) {materialFunction=&Sprite}.World position offset
    B=MaterialFunctionCall(Offset Amount=@n74) {materialFunction=/Engine/Functions/Engine_MaterialFunctions02/WorldPositionOffset/CameraOffset}.Camera Offset
   )
   False=StaticSwitchParameter(True=LinearInterpolate(B=@n77, Alpha=@n80) {constB=0}, False=Constant()) {parameterName=Distance Scale}
  ) {defaultValue=true, parameterName=CameraFacingMeshSprite}
  n82 = Add(A=@n81, B=WorldPosition().XYZ)
  StaticSwitchParameter(True=Add(A=RotateAboutAxis(NormalizedRotationAxis=@n60, RotationAngle=@n63, PivotPoint=@n65, Position=@n82), B=@n81), False=@n81) {defaultValue=true, parameterName=RotateAboutAxis}
  TextureCoordinate()
  WorldPosition()
 ]
}
